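; BPI_MusicEncounterComponent_Boss is a Blueprint interface for boss components in the music encounter system.
; It declares one function, UsesTargetChangedReplication, which returns a boolean that reads false here.
; The name suggests the flag reports whether a boss component replicates target changes, but the file holds no logic that uses it.
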(fn UsesTargetChangedReplication
  (return false))
